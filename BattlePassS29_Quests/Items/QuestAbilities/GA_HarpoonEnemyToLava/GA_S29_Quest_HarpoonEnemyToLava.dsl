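; Season 29 quest ability that detects the avatar harpooning an opposing player into lava.
; It waits for a HappyGhost (harpoon) hit event, then checks that the target is a player with the required opponent tags.
; Chained async event waits watch for lava-tagged damage, broadcasting the HarpoonedPlayerIntoLava custom verb and gameplay event.
; The ability ends if the target lands, fails a check, or takes non-lava damage.
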(event Ability|EventActivateAbility
  (bind _returnvalue (Ability|GetAvatarActorfromActorInfo))
  (Utilities|IsValid _returnvalue
    (:Is%20Valid
      (bind (_asynctaskproxy _payload) (Ability|Async|WaitGameplayEventToActor _returnvalue "(TagName=\"Event.Athena.Item.HappyGhost.Hit\")" false false)
        (:EventReceived
          (Development|PrintString "HarpoonGAFired")
          (bind (_payload_eventtag _payload_instigator _payload_target) (Utilities|Struct|BreakGameplayEventData _payload))
          (bind _asfort_player_pawn (Utilities|Casting|CastToFortPlayerPawn _payload_target)
            (:then
              (bind _output_get (Variables|Default|SetTarget%20Opponent _asfort_player_pawn))
              (Ability|Async|WaitGameplayEventToActor _payload_target "(TagName=\"Event.Physics.Landed\")" false false
                (:then
                  (if (GameplayTags|HasAllMatchingGameplayTags _output_get (Variables|Default|GetRequired%20Opponent%20Tags))
                    (Development|PrintString "Player Hit with Harpoon")
                    (bind (_asynctaskproxy_2 _payload_2) (Ability|Async|WaitGameplayEventToActor _payload_target "(TagName=\"Event.Damage.Received\")" false false)
                      (:EventReceived
                        (bind (_payload_eventtag_2 _payload_instigator_2 _payload_target_2 _payload_optionalobject _payload_optionalobject2 _payload_contexthandle _payload_instigatortags) (Utilities|Struct|BreakGameplayEventData _payload_2))
                        (if (GameplayTags|HasAnyTags _payload_instigatortags (Variables|Default|GetLava%20Container))
                          (Development|PrintString "Ouchie Fire" true true "(R=1.000000,G=0.347983,B=0.000000,A=1.000000)")
                          (CustomVerb|BroadcastCustomVerbMessage (CustomVerb|MakeCustomVerbMessage (Fortnite|GetFortPlayerControllerfromActor (CallFunction|GetAvatarasFortPlayerPawn)) "(TagName=\"CustomVerb.HarpoonedPlayerIntoLava\")"))
                          (Ability|SendGameplayEventtoActor 0 "(TagName=\"CustomVerb.HarpoonedPlayerIntoLava\")")
                          (else
                            (CallFunction|StopListeningforLavaDamage)))))
                    (else
                      (CallFunction|StopListeningforLavaDamage))))
                (:EventReceived
                  (Development|PrintString "Gotta hit me again with a harpoon")
                  (CallFunction|StopListeningforLavaDamage))))
            (:CastFailed
              (Development|PrintString "Not a player")
              (CallFunction|StopListeningforLavaDamage))))))))

(event Custom|StopListeningForLavaDamage
  (Ability|EndAbility))
